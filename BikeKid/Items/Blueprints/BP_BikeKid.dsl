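; BP_BikeKid is the player-guided missile pawn deployed from BP_BikeKidWeapon, with a RideBox, health and a launch cue.
; BeginPlay briefly ignores owner collision, links the owner and weapon, and sets up hitbox, aim assist, camera ability, audio, team-colored VFX and quality-gated rotor wash.
; Input events roll the mesh, while damage events drive crinkle materials and motor audio.
; Dismissal, destruction and EndPlay clear indicators, sounds and wash force; only hostiles count as auto-fire targets.
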(event EventBeginPlay
  (bind _returnvalue (RemoteControl|Player|GetControllingPlayerPawn))
  ;[1 1 1] Sequence just for organizational purposes
  (Utilities|FlowControl|Sequence
    (:then_0
      ; We disable the collision with the owner for s short time after the spawn to avoid possible collision issues during the deploy.
      (CallFunction|ToggleCollisionwithOwningPawn false)
      (Utilities|FlowControl|Delay 1.0)
      (CallFunction|ToggleCollisionwithOwningPawn true))
    (:then_1
      ;[1 1 1] Var set up & Hotfixes
      (CallFunction|SetPlayerPawnVariablesInternal)
      (bind _asfort_bike_kid_movement_component (Utilities|Casting|CastToFortBikeKidMovementComponent (Pawn|GetMovementComponent))
        (:then
          (Class|FortBikeKidMovementComponent|SetHitBoxComponent _asfort_bike_kid_movement_component (Variables|Default|GetHitBox)))
        (:CastFailed
          _))
      ;;[0.1168 1 0.08209] Aim Assist
      (AimAssist|RegisterAimAssistTarget 0 (Variables|Character|GetCapsuleComponent))
      ;[1 1 1] Force Feedback on spawn
      (ForceFeedback|SpawnForceFeedbackAttached "/Game/Weapons/ForceFeedback/FF_Shotgun_Flip.FF_Shotgun_Flip" (Variables|Character|GetCapsuleComponent)))
    (:then_2
      ;[1 1 1] Camera pulled out on start, then pull back in to default camera class
      (Utilities|FlowControl|Delay 1.0)
      (if (not (Variables|FortBikeKid|GetbExploded))
        (if (not (Networking|HasAuthority))
          (Abilities|TryActivateAbilitiesbyTag (Ability|GetAbilitySystemComponent self) (GameplayTags|MakeGameplayTagContainerfromTag (Variables|Default|GetCameraDefaultTag)) true))))
    (:then_3
      ;[0 0.2435 1] Audio
      (Networking|CanExecuteCosmeticEvents
        (:True
          (Audio|SpawnSoundatLocation "/BikeKid/Sounds/Guided_Missile/GuidedMissile_Launch_Cue.GuidedMissile_Launch_Cue" (Transformation|GetActorLocation))
          (Audio|SpawnSoundAttached :Sound "/BikeKid/Sounds/Guided_Missile/GuidedMissile_InAir_Banking_Loop_Cue.GuidedMissile_InAir_Banking_Loop_Cue" :AttachToComponent (Variables|Default|GetSound_Flying_Loop) :AttachPointName "None" :LocationType "KeepRelativeOffset" :bStopWhenAttachedToDestroyed true :VolumeMultiplier 1.0 :PitchMultiplier 1.0 :StartTime 0.0 :bAutoDestroy true)
          (CallFunction|SetAffiliationAudio)
          (Utilities|Time|SetTimerbyFunctionName 0 "SetupGameplayAudioParameters" 0.1 true))))
    (:then_4
      ;[0.5104 0.02083 1] VFX
      (Networking|CanExecuteCosmeticEvents
        (:True
          (Utilities|IsValid _returnvalue
            (:Is%20Valid
              (bind _asbp_bike_kid_weapon (Utilities|Casting|CastToBP_BikeKidWeapon (Variables|/Script/FortniteGame.FortPawn|GetCurrentWeapon _returnvalue))
                (:then
                  (Components|Activation|Activate (Variables|/BikeKid/Items/Blueprints/BP_BikeKidWeapon.BP_BikeKidWeapon_C|GetNS_Javelin_RearMuzzle _asbp_bike_kid_weapon) false))
                (:CastFailed
                  _))
              (Effects|Components|ParticleSystem|SetColorParameter (Variables|Default|GetPS_BlinkingLight) "TeamColor" (select (Team|OnSameTeam self (Utilities|Array|Get(acopy) (Fortnite|GetLocalFortPlayerControllers true))) "(R=0.000000,G=0.200000,B=0.800000,A=1.000000)" "(R=1.000000,G=0.000000,B=0.000000,A=1.000000)"))
              (if (>= (Settings|GetVisualEffectQuality (Settings|GetFortGameUserSettings)) 3)
                (Utilities|IsValid (Variables|Default|GetPhysical%20Wash%20Force)
                  (:Is%20Not%20Valid
                    (bind _returnvalue_2 (Game|SpawnActorfromClass "/Game/Effects/Niagara/Blueprints/PhsyicsFields/FN_WashForce.FN_WashForce_C" (Transformation|GetActorTransform self) "Undefined" "MultiplyWithRoot" 0 3e+02 (Variables|Default|GetRotorWashForceStrength) -999.0 0))
                    (bind _output_get (Variables|Default|SetPhysical%20Wash%20Force _returnvalue_2))
                    (Transformation|AttachActorToComponent _output_get (Variables|Character|GetMesh) "None" "SnapToTarget" "KeepRelative" "KeepRelative" true))) ; Don't spawn the Rotor Wash multiple times
              ))))))
    (:then_5
      ;[1 1 1] UI Feedback
      (CallFunction|SetTeamMatesIndicator true))))

(event EventDestroyed
  ;[1 1 1] UI Feedback
  (|CallParentFunction)
  (CallFunction|SetTeamMatesIndicator false)
  (MiniMapData|SetMiniMapIndicatorIsVisible (Variables|Default|GetFortMiniMapComponent) false))

(event EventBP_OnBikeKidDismissed
  (params (DismissalReason :type Gameplay%20Tag%20Structure%20%28by%20ref%29))
  ; Dismiss
  (MiniMapData|SetMiniMapIndicatorIsVisible (Variables|Default|GetFortMiniMapComponent) false)
  (Networking|SwitchHasAuthority
    (:Authority
      (bind _asfort_ability_system_component (Utilities|Casting|CastToFortAbilitySystemComponent (Ability|GetAbilitySystemComponent self)))
      (Tags|CancelAbilitiesWithTags _asfort_ability_system_component (GameplayTags|MakeGameplayTagContainerfromTag (Variables|Default|GetCameraDefaultTag))))))

(event EventBP_OnDestructionStarted
  (bind _physical_wash_force (Variables|Default|GetPhysical%20Wash%20Force))
  ; Destruction start
  (Networking|CanExecuteCosmeticEvents
    (:True
      (SoundIndicator|StopSoundIndicatorLooping (Variables|Default|GetFortSoundIndicator))
      (Utilities|IsValid _physical_wash_force
        (:Is%20Valid
          (Actor|DestroyActor _physical_wash_force))))))

(event EventBP_OnBikeKidLaunched
  ; Launch
  (Networking|SwitchHasAuthority
    (:Authority
      (GameplayCue|ExecuteGameplayCue self (Variables|Default|GetGameplay%20Cue%20Launched)))
    (:Remote
      (CallFunction|ShowHUDKeybinds))))

(event EventEndPlay
  (params (EndPlayReason :type EEndPlayReason%20Enum :default "Destroyed"))
  (bind _physical_wash_force (Variables|Default|GetPhysical%20Wash%20Force))
  (Utilities|Time|ClearTimerbyFunctionName 0 "SetupGameplayAudioParameters")
  (Audio|Components|Audio|Stop (Variables|Components|GetBikeKidMotorLoopComponent))
  (Utilities|IsValid _physical_wash_force
    (:Is%20Valid
      (Actor|DestroyActor _physical_wash_force))) ; Checks if Rotor Wash Actor is Valid
)

(event EventOnRotationInputReceived
  (params (RotationInput :type Rotator%20%28by%20ref%29 :default "0, 0, 0"))
  ; Handles the missile roll rotation
  (Transformation|SetRelativeRotation (Variables|Character|GetMesh) (Math|Rotator|MakeRotator (.roll (Variables|/Script/Engine.SceneComponent|GetRelativeRotation (Variables|Character|GetMesh))) (Math|Interpolation|FInterpTo (.pitch (Variables|/Script/Engine.SceneComponent|GetRelativeRotation (Variables|Character|GetMesh))) (* (.yaw RotationInput) 15.0) (Utilities|Time|GetWorldDeltaSeconds) 0.6) (.yaw (Variables|/Script/Engine.SceneComponent|GetRelativeRotation (Variables|Character|GetMesh)))) false false))

(event EventOnDeathPlayEffects
  (params (Damage :type Float%20%28single-precision%29) (DamageTags :type Gameplay%20Tag%20Container%20Structure%20%28by%20ref%29) (Momentum :type Vector :default "0, 0, 0") (HitInfo :type Hit%20Result%20Structure%20%28by%20ref%29) (InstigatedBy :type Fort%20Pawn%20Object%20Reference) (DamageCauser :type Actor%20Object%20Reference) (EffectContext :type Gameplay%20Effect%20Context%20Handle%20Structure))
  (Audio|Components|Audio|FadeOut (Variables|Components|GetBikeKidMotorLoopComponent) 1.0 0.0 "Linear" :!class /Script/Engine.AudioComponent))

(event EventOnDamagePlayEffects
  (params (Damage :type Float%20%28single-precision%29) (DamageTags :type Gameplay%20Tag%20Container%20Structure%20%28by%20ref%29) (Momentum :type Vector :default "0, 0, 0") (HitInfo :type Hit%20Result%20Structure%20%28by%20ref%29) (InstigatedBy :type Fort%20Pawn%20Object%20Reference) (DamageCauser :type Actor%20Object%20Reference) (EffectContext :type Gameplay%20Effect%20Context%20Handle%20Structure))
  (Utilities|FlowControl|Sequence
    (:then_0
      (Default|SendDamageValuetoMaterials Damage))
    (:then_1
      (Utilities|FlowControl|Delay 0.2)
      (Audio|Parameter|SetFloatParameter (Variables|Components|GetBikeKidMotorLoopComponent) "DamageScale" (Stats|GetHealthPercentage) :!class /Script/AudioExtensions.AudioParameterControllerInterface))))

(fn DebugPrintSpeed
  (Development|PrintString (Utilities|String|Append "Max Fly Speed::  " (Utilities|String|ToString(Float) (Variables|/Script/Engine.CharacterMovementComponent|GetMaxFlySpeed (Variables|Character|GetCharacterMovement)))) true true "(R=0.092793,G=1.000000,B=0.156864,A=1.000000)" 5.0))

(fn GetAffiliationsToShowFor
  (return (Utilities|Array|MakeArray (Utilities|Enum|LiteralenumEFortTeamAffiliation "Hostile"))))

(fn GetTargetOwner
  (return 0))

(fn IsValidAutoFireTarget
  (params (Attacker :type Actor%20Object%20Reference))
  (return (Utilities|Enum|Equal(Enum) (Team|GetTeamAffiliationBetweenTeams (Team|GetActorTeam Attacker) (Team|GetActorTeam self)) "Hostile")))

(fn SendDamageValueToMaterials
  (params (ParameterValue :type Float%20%28double-precision%29))
  (bind _mesh (Variables|Character|GetMesh))
  (bind _healthpctmissingtoshowdmgcrinkle (Variables|Components|GetHealthPctMissingToShowDmgCrinkle))
  ;[1 1 1] Only Show damage crinkles after health is below a certain threshold
  (if (> ParameterValue _healthpctmissingtoshowdmgcrinkle)
    (bind _output_get (Variables|Components|SetSavedDamageValueForMID (Math|Float|Lerp (* (* _healthpctmissingtoshowdmgcrinkle (Variables|Components|GetMaxCrinkleSeverity)) -1.0) (Variables|Components|GetMaxCrinkleSeverity) ParameterValue)))
    (Development|PrintString (Utilities|String|ToString(Float) _output_get))
    (Rendering|Material|SetScalarParameterValueonMaterials _mesh "Damage" _output_get)
    (else
      (bind _output_get_2 (Variables|Components|SetSavedDamageValueForMID 0.0))
      (Development|PrintString (Utilities|String|ToString(Float) _output_get_2))
      (Rendering|Material|SetScalarParameterValueonMaterials _mesh "Damage" _output_get_2) ; Need to reset crinkle back to 0.0 if we're healed beyond the min threshold
    )))

(fn SetPlayerPawnVariables_Internal
  (bind _asfort_player_pawn_athena (Utilities|Casting|CastToFortPlayerPawnAthena (Game|GetInstigator)))
  (bind _asbp_bike_kid_weapon (Utilities|Casting|CastToBP_BikeKidWeapon (Variables|/Script/FortniteGame.FortPawn|GetCurrentWeapon _asfort_player_pawn_athena)))
  (Variables|Default|SetOwningPawn _asfort_player_pawn_athena)
  (Variables|Default|SetWID_Controller _asbp_bike_kid_weapon)
  (Class|BPBikeKidWeapon|OnBikeKidDeploy _asbp_bike_kid_weapon))

(fn ShouldAlliesAutoFireOnDeployedTarget
  (return false))

(fn ShouldAutoFireOnSelfDeployedTarget
  (return false))

(fn ShouldShowSoundIndicator
  (params (PlayerController :type Fort%20Player%20Controller%20Object%20Reference))
  (return true))

(fn ShowHUDKeybinds
  (WeaponHUDData|SetHudKeyActionsVisibility (Variables|Default|GetWID_Controller) (Utilities|Array|MakeArray (Utilities|Struct|MakeWeaponHudKeyActionVisibility 0) (Utilities|Struct|MakeWeaponHudKeyActionVisibility 1 true))))

(fn ToggleCollisionWithOwningPawn
  (params (Enabled :type Boolean))
  (bind _returnvalue (Game|GetInstigator))
  (Collision|IgnoreActorwhenMoving (Variables|Default|GetRideBox) _returnvalue (not (Variables|GetEnabled)))
  (bind _asfort_pawn (Utilities|Casting|CastToFortPawn _returnvalue))
  (Collision|IgnoreActorwhenMoving (Variables|/Script/Engine.Character|GetCapsuleComponent _asfort_pawn) self (not (Variables|GetEnabled))))

(fn ConstructionScript)
