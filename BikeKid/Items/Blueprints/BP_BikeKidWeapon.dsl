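; Fortnite weapon Blueprint for the BikeKid item, managing its deploy HUD key prompts and the visibility of its "Rocket" mesh bone.
; On cosmetic-capable clients, the bone hides on deploy or when equipped without ammo, and reappears when a reload starts.
; SetBikeKidHUDActionKeys can show the second key action outside vehicles, but equip and both vehicle-change branches here always pass true, hiding both actions.
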(event EventBeginPlay
  (|CallParentFunction)
  (bind _asfort_player_pawn (Utilities|Casting|CastToFortPlayerPawn (Game|GetPlayerPawn 0)))
  (Vehicle|BindEventtoOnVehicleStateChangedEvent _asfort_player_pawn (delegate Custom|OnVehicle)))

(event Custom|OnVehicle
  (params (FortPlayerPawn :type Fort%20Player%20Pawn%20Object%20Reference) (NewVehicle :type Actor%20Object%20Reference) (OldVehicle :type Actor%20Object%20Reference))
  (Utilities|IsValid NewVehicle
    (:Is%20Valid
      (CallFunction|SetBikeKidHUDActionKeys true))
    (:Is%20Not%20Valid
      (CallFunction|SetBikeKidHUDActionKeys true))))

(event FortWeapon|EventOnEquippedAndAttached
  (CallFunction|SetBikeKidHUDActionKeys true)
  (Networking|CanExecuteCosmeticEvents
    (:True
      (EventDispatchers|BindEventtoOnReloadStarted (delegate Custom|OnLocalReloadStarted_Event))
      (if (<= (FortWeapon|GetLocalRemainingAmmo) 0)
        (Components|SkinnedMesh|HideBonebyName (Variables|Visual|GetWeaponMesh) "Rocket" "PBO_None")))))

(event Custom|OnLocalReloadStarted_Event
  (params (ReloadTime :type Float%20%28single-precision%29) (ReloadType :type EFortWeaponReloadType%20Enum :default "ReloadWholeClip"))
  (Components|SkinnedMesh|UnHideBonebyName (Variables|Visual|GetWeaponMesh) "Rocket"))

(fn OnBikeKidDeploy
  (Networking|CanExecuteCosmeticEvents
    (:True
      (Components|SkinnedMesh|HideBonebyName (Variables|Visual|GetWeaponMesh) "Rocket" "PBO_None"))))

(fn SetBikeKidHUDActionKeys
  (params (bHideDeploy :type Boolean))
  ;[1 1 1] Set HUD keys
  (if (not bHideDeploy)
    (bind _asfort_player_pawn (Utilities|Casting|CastToFortPlayerPawn (Game|GetPlayerPawn 0))
      (:then
        (if (not (Vehicle|IsinVehicle _asfort_player_pawn))
          (WeaponHUDData|SetHudKeyActionsVisibility (Utilities|Array|MakeArray (Utilities|Struct|MakeWeaponHudKeyActionVisibility 0) (Utilities|Struct|MakeWeaponHudKeyActionVisibility 1 true)))))
      (:CastFailed
        _))
    (WeaponHUDData|SetHudKeyActionsVisibility (Utilities|Array|MakeArray (Utilities|Struct|MakeWeaponHudKeyActionVisibility 0) (Utilities|Struct|MakeWeaponHudKeyActionVisibility 1)))))

(fn ConstructionScript)
