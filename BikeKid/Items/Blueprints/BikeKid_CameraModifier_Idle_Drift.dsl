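; Camera modifier that adds a slow idle drift to the player camera's position, plus an FOV offset.
; Separate position and FOV alphas ramp over their blend times; BlendOut reverses them, and the modifier removes itself once the position alpha reaches zero.
; A corner-index state machine steps right and height offsets by a per-update increment between limits in a repeating pattern.
; View-relative offsets scale by alpha; FOV also scales by TargetingAlpha.
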(event Custom|BlendOut
  (Variables|Default|SetBlendIn false))

(fn BlueprintModifyCamera
  (params (DeltaTime :type Float%20%28single-precision%29) (ViewLocation :type Vector :default "0, 0, 0") (ViewRotation :type Rotator :default "0, 0, 0") (FOV :type Float%20%28single-precision%29))
  (Utilities|FlowControl|Sequence
    (:then_0
      ;[1 1 1] Remove the Camera Modifier once the "Blend out" is complete
      (if (Variables|Default|GetBlendIn)
        _
        (elif (== (Variables|Default|GetCameraOffsetAlpha) 0.0)
          (Game|Player|RemoveCameraModifier (Game|GetPlayerCameraManager 0) self))))
    (:then_1
      ;[1 1 1] Create a Alpha that interpolates from 0 - 1
      (Variables|Default|SetCameraOffsetAlpha (Math|Float|Clamp(Float) (+ (Math|Float|float*float DeltaTime (/ 1.0 (Math|Float|SelectFloat (Variables|Default|GetCameraOffsetBlendInSeconds) (Variables|Default|GetCameraOffsetBlendOutSeconds) (Variables|Default|GetBlendIn))) (Math|Float|SelectFloat 1.0 -1.0 (Variables|Default|GetBlendIn))) (Variables|Default|GetCameraOffsetAlpha)))))
    (:then_2
      (Variables|Default|SetFovOffsetAlpha (Math|Float|Clamp(Float) (+ (Math|Float|float*float DeltaTime (/ 1.0 (Math|Float|SelectFloat (Variables|Default|GetFovOffsetBlendInSeconds) (Variables|Default|GetFovOffsetBlendOutSeconds) (Variables|Default|GetBlendIn))) (Math|Float|SelectFloat 1.0 -1.0 (Variables|Default|GetBlendIn))) (Variables|Default|GetFovOffsetAlpha))))
      ;[1 1 1] Up Right
      (if (== 0 (Variables|Default|GetOffset_Corner_Int))
        (Variables|Default|SetRightOffset (Math|Float|Clamp(Float) (+ (Variables|Default|GetRightOffset) (Variables|Default|GetOffset_Increment)) (Variables|Default|GetOffset_Min) (Variables|Default|GetOffset_Max)))
        (Variables|Default|SetHeightOffset (Math|Float|Clamp(Float) (+ (Variables|Default|GetHeightOffset) (Variables|Default|GetOffset_Increment)) (Variables|Default|GetOffset_Min) (Variables|Default|GetOffset_Max)))
        (if (>= (Variables|Default|GetRightOffset) (Variables|Default|GetOffset_Max))
          (Math|Integer|IncrementInt (Variables|Default|GetOffset_Corner_Int)
            (:%20%20
              _)))
        (elif (== 1 (Variables|Default|GetOffset_Corner_Int))
          ;[1 1 1] Down Left
          (Variables|Default|SetRightOffset (Math|Float|Clamp(Float) (- (Variables|Default|GetRightOffset) (Variables|Default|GetOffset_Increment)) (Variables|Default|GetOffset_Min) (Variables|Default|GetOffset_Max)))
          (Variables|Default|SetHeightOffset (Math|Float|Clamp(Float) (- (Variables|Default|GetHeightOffset) (Variables|Default|GetOffset_Increment)) (Variables|Default|GetOffset_Min) (Variables|Default|GetOffset_Max)))
          (if (<= (Variables|Default|GetRightOffset) (Variables|Default|GetOffset_Min))
            (Math|Integer|IncrementInt (Variables|Default|GetOffset_Corner_Int)
              (:%20%20
                _)))
          (elif (== 2 (Variables|Default|GetOffset_Corner_Int))
            ;[1 1 1] Up Left; check height offset since we're not incrementing right
            (Variables|Default|SetRightOffset (Math|Float|Clamp(Float) (+ (Variables|Default|GetRightOffset) (Variables|Default|GetOffset_Increment)) (Variables|Default|GetOffset_Min) (Variables|Default|GetOffset_Max)))
            (Variables|Default|SetHeightOffset (Math|Float|Clamp(Float) (+ (Variables|Default|GetHeightOffset) (Variables|Default|GetOffset_Increment)) (Variables|Default|GetOffset_Min) (Variables|Default|GetOffset_Max)))
            (if (>= (Variables|Default|GetHeightOffset) (Variables|Default|GetOffset_Max))
              (Math|Integer|IncrementInt (Variables|Default|GetOffset_Corner_Int)
                (:%20%20
                  _)))
            (elif (== 3 (Variables|Default|GetOffset_Corner_Int))
              ;[1 1 1] Down RIght
              (Variables|Default|SetRightOffset (Math|Float|Clamp(Float) (+ (Variables|Default|GetRightOffset) (Variables|Default|GetOffset_Increment)) (Variables|Default|GetOffset_Min) (Variables|Default|GetOffset_Max)))
              (Variables|Default|SetHeightOffset (Math|Float|Clamp(Float) (- (Variables|Default|GetHeightOffset) (Variables|Default|GetOffset_Increment)) (Variables|Default|GetOffset_Min) (Variables|Default|GetOffset_Max)))
              (if (<= (Variables|Default|GetHeightOffset) (Variables|Default|GetOffset_Min))
                (Math|Integer|IncrementInt (Variables|Default|GetOffset_Corner_Int)
                  (:%20%20
                    _)))
              (elif (== 4 (Variables|Default|GetOffset_Corner_Int))
                ;[1 1 1] Reset the camera positoin, then reset the loop
                (Variables|Default|SetRightOffset (Math|Float|Clamp(Float) (- (Variables|Default|GetRightOffset) (Variables|Default|GetOffset_Increment)) (Variables|Default|GetOffset_Min) (Variables|Default|GetOffset_Max)))
                (Variables|Default|SetHeightOffset (Math|Float|Clamp(Float) (+ (Variables|Default|GetHeightOffset) (Variables|Default|GetOffset_Increment)) (Variables|Default|GetOffset_Min) (Variables|Default|GetOffset_Max)))
                (if (>= (Variables|Default|GetHeightOffset) (Variables|Default|GetBass_Offset))
                  (Variables|Default|SetOffset_Corner_Int 0)))))))
      (return (Utilities|Operators|Add ViewLocation (Utilities|Operators|Multiply (Math|Vector|GetRightVector ViewRotation) (Variables|Default|GetRightOffset) (Variables|Default|GetCameraOffsetAlpha)) (Utilities|Operators|Multiply (Math|Vector|GetUpVector ViewRotation) (Variables|Default|GetHeightOffset) (Variables|Default|GetCameraOffsetAlpha)) (Utilities|Operators|Multiply (Math|Vector|GetForwardVector ViewRotation) (Variables|Default|GetForwardOffset) (Variables|Default|GetCameraOffsetAlpha))) ViewRotation (+ FOV (Utilities|Operators|Multiply (Math|Float|Lerp (Variables|Default|GetMinAdditionalFOV) (Variables|Default|GetMaxAdditionalFOV) 1.0) (Variables|Default|GetFovOffsetAlpha) (Variables|Default|GetTargetingAlpha)))))))
