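; Gameplay ability that activates the BikeKid item, apparently a remote-controlled pawn, on the player.
; On activation it applies an action-blocking effect, plays a controller-idle montage chosen by body type and gender, and updates quickbar, input context and reticle tags.
; It then sends BikeKid activation and RC pawn fired gameplay events, plus a quest stat event.
; A helper checks whether failure tags warrant a failed montage.
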(event Ability|EventActivateAbility
  (bind _returnvalue (Ability|GetAvatarActorfromActorInfo))
  (bind _returnvalue_2 (Ability|GetAvatarActorfromActorInfo))
  (bind _returnvalue_3 (Ability|GetOwningActorfromActorInfo))
  (bind _returnvalue_4 (Fortnite|GetFortPlayerControllerfromActor _returnvalue))
  (Ability|ApplyGameplayEffectSpecToOwner (Ability|MakeOutgoingGameplayEffectSpec (Variables|Default|GetBlockActionsGameplayEffectClass)))
  (bind _asfort_player_pawn (Utilities|Casting|CastToFortPlayerPawn _returnvalue_2))
  (CallFunction|SetPlayerAnimMontageBasedOnCharacter _asfort_player_pawn)
  (Animation|Utilities|PlayAnimMontage _returnvalue_2 (Variables|Default|GetPlayer%20Anim%20Montage))
  (bind _asfort_player_controller_gameplay (Utilities|Casting|CastToFortPlayerControllerGameplay (Fortnite|GetFortPlayerControllerfromActor (Ability|GetAvatarActorfromActorInfo))))
  (HUD|SetHUDElementVisibilitywithContext _asfort_player_controller_gameplay (GameplayTags|MakeGameplayTagContainerfromTag "(TagName=\"HUD.Bacchus.Quickbar.Bar\")") true "BikeKid" "")
  ;[1 1 1] Update context for HUD Layout
  (Class|FortControllerComponentInputContextTracker|UpdateContexts (Class|FortPlayerController|GetContextTrackerComponent (Fortnite|GetFortPlayerControllerfromActor (Ability|GetAvatarActorfromActorInfo))) (Variables|Default|GetContextTags) (GameplayTags|MakeGameplayTagContainerfromArray (Utilities|Array|MakeArray)) false)
  (Ability|GameplayEffect|AddLooseGameplayTags _returnvalue_3 (GameplayTags|MakeGameplayTagContainerfromTag "(TagName=\"HUD.Athena.Reticle.HideMuzzleBlocked\")"))
  (Ability|SendGameplayEventtoActor _returnvalue_3 "(TagName=\"Event.BikeKid.Activate\")" (Utilities|Struct|MakeGameplayEventData "(TagName=\"Event.BikeKid.Activate\")" _returnvalue_3))
  (Ability|SendGameplayEventtoActor _returnvalue_3 "(TagName=\"Event.Weapon.Fired.RCPawn\")" (Utilities|Struct|MakeGameplayEventData "(TagName=\"Event.Weapon.Fired.RCPawn\")" _returnvalue_3))
  ;[0.1085 1 0.1109] Activate BikeKid Quest
  (Utilities|IsValid _returnvalue_4
    (:Is%20Valid
      (Quest|SendComplexCustomStatEvent (Quests|GetQuestManager _returnvalue_4 "Invalid") _returnvalue "(GameplayTags=((TagName=\"Athena.Quest.BikeKid.Activate\")))" "(GameplayTags=)" 1))))

(fn SetPlayerAnimMontageBasedonCharacter
  (params (Player%20Pawn :type Fort%20Player%20Pawn%20Object%20Reference))
  (bind _ismale (FortPlayerPawn|GetGender Player%20Pawn))
  ;[1 1 1] Check against Large body type & gender to play correct anim for controller idle
  (switch Utilities|FlowControl|Switch|SwitchonEFortCustomBodyType (Customization|GetCharacterBodyType Player%20Pawn)
    (:Small
      _)
    (:Medium
      _)
    (:Large
      (if _ismale
        (Variables|Default|SetPlayer%20Anim%20Montage (Variables|Default|GetPlayer%20Anim%20Montage%20Large%20Body%20Type))
        (else
          (Variables|Default|SetPlayer%20Anim%20Montage (Variables|Default|GetPlayer%20Anim%20Montage%20Large%20Body%20Type%20Female))))
      (return)))
  (if (not _ismale)
    (Variables|Default|SetPlayer%20Anim%20Montage (Variables|Default|GetPlayer%20Anim%20Montage%20Female))))

(fn ShouldPlayFailedMontage
  (params (FailedReason :type Gameplay%20Tag%20Container%20Structure%20%28by%20ref%29))
  (return (GameplayTags|HasAnyTags FailedReason (Variables|Default|GetFailedReasonsThatPlayMontage) true)))
