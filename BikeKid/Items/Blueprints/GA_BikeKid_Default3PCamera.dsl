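; GA_BikeKid_Default3PCamera is a gameplay ability that applies the default third-person camera for the BikeKid item.
; On activation it calls the parent activation, then pushes a camera mode using the class stored in its Camera Mode Class variable.
; The end-ability and failed-passive-activation events are present but empty, so this graph has no logic to remove the pushed mode.
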(event Ability|EventActivateAbility
  (|CallParentFunction)
  (Camera|PushCameraMode (Variables|Default|GetCamera%20Mode%20Class)))

(event Ability|EventOnEndAbility
  (params (bWasCancelled :type Boolean)))

(event Ability|EventFailedToActivatePassiveAbility
  (params (ActorInfo :type Gameplay%20Ability%20Actor%20Info%20Structure)))
