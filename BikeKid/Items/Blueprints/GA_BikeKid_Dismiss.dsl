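; Gameplay ability that runs when BikeKid is dismissed, undoing its HUD and input state, then ending immediately.
; On activation it updates the controller's input contexts, stops the configured montage if current, and resets quickbar visibility under the BikeKid context.
; It also hides the BikeKid key bindings through the pawn's current BikeKid weapon and removes the loose tag that hides the muzzle-blocked reticle.
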(event Ability|EventActivateAbility
  (bind _returnvalue (Ability|GetAvatarActorfromActorInfo))
  ;[1 1 1] Update context for HUD Layout
  (Class|FortControllerComponentInputContextTracker|UpdateContexts (Class|FortPlayerController|GetContextTrackerComponent (Fortnite|GetFortPlayerControllerfromActor (Ability|GetOwningActorfromActorInfo))) (GameplayTags|MakeGameplayTagContainerfromArray (Utilities|Array|MakeArray)) (Variables|Default|GetContextTags) false)
  (Animation|Utilities|StopMontageifCurrent _returnvalue (Variables|Default|GetStop%20Anim%20Montage))
  (bind _asfort_player_controller_gameplay (Utilities|Casting|CastToFortPlayerControllerGameplay (Fortnite|GetFortPlayerControllerfromActor _returnvalue)))
  (HUD|SetHUDElementVisibilitywithContext _asfort_player_controller_gameplay (GameplayTags|MakeGameplayTagContainerfromTag "(TagName=\"HUD.Bacchus.Quickbar.Bar\")") false "BikeKid" "")
  (CallFunction|HideHUDKeyBindings)
  (Ability|GameplayEffect|RemoveLooseGameplayTags (Ability|GetOwningActorfromActorInfo) (GameplayTags|MakeGameplayTagContainerfromTag "(TagName=\"HUD.Athena.Reticle.HideMuzzleBlocked\")"))
  (Ability|EndAbility))

(fn HideHUDKeyBindings
  (bind _outpawn (Class|AbilityFunctions|GetPawnfromPlayerStateZone (Ability|GetOwningActorfromActorInfo)))
  (bind _asfort_player_pawn_athena (Utilities|Casting|CastToFortPlayerPawnAthena _outpawn))
  (bind _asbp_bike_kid_weapon (Utilities|Casting|CastToBP_BikeKidWeapon (Variables|/Script/FortniteGame.FortPawn|GetCurrentWeapon _asfort_player_pawn_athena)))
  ;[1 1 1] Hide BikeKid controls on dismiss
  (Class|BPBikeKidWeapon|SetBikeKidHUDActionKeys _asbp_bike_kid_weapon true))
