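# Sound cue for firing the BikeKid launcher weapon, built from reused guided missile and rocket launcher sounds.
# A mixer layers a random pick of three guided missile launches on close attenuation with two rocket launcher medium and distant shots on medium attenuation, the distant one at half volume.
# Each layer overrides its sound class as third-person or distant fire, and the cue references the guided missile fire concurrency set.
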SoundCue /BikeKid/Items/Sound/Weapon_BikeKid_Launcher_Fire_Cue
root=Mixer_0
nodes [
 {
  id=Attenuation_0
  children [WavePlayer_6]
  properties {attenuationSettings=&Fort_Weapon_Projectile_Medium_Mono_Athena_Att}
 }
 {
  id=Attenuation_1
  children [WavePlayer_7]
  properties {attenuationSettings=&Fort_Weapon_Projectile_Medium_Mono_Athena_Att}
 }
 {id=Attenuation_9, children [Modulator_6], properties {attenuationSettings=&Fort_Weapon_Projectile_Close_Athena_Att}}
 {id=Mixer_0, children [SoundClass_1,SoundClass_0,SoundClass_2], properties {inputVolume [1,1,0.5]}}
 {id=Modulator_6, children [Random_1], properties {pitchMax=1, pitchMin=1}}
 {id=Random_1, children [WavePlayer_0,WavePlayer_2,WavePlayer_3], properties {weights [1,1,1]}}
 {
  id=SoundClass_0
  children [Attenuation_0]
  properties {soundClassOverride=&Fort_WeaponRanged_Fire_Distant}
 }
 {
  id=SoundClass_1
  children [Attenuation_9]
  properties {soundClassOverride=&Fort_WeaponRanged_Fire_3P_PSM}
 }
 {
  id=SoundClass_2
  children [Attenuation_1]
  properties {soundClassOverride=&Fort_WeaponRanged_Fire_Distant}
 }
 {id=WavePlayer_0, properties {soundWaveAssetPtr=/Game/Athena/Sounds/Weapons/Guided_Missile/GuidedMissile_Launch_01}}
 {id=WavePlayer_2, properties {soundWaveAssetPtr=/Game/Athena/Sounds/Weapons/Guided_Missile/GuidedMissile_Launch_02}}
 {id=WavePlayer_3, properties {soundWaveAssetPtr=/Game/Athena/Sounds/Weapons/Guided_Missile/GuidedMissile_Launch_03}}
 {id=WavePlayer_6, properties {soundWaveAssetPtr=/Game/Sounds/Fort_Weapons/Launchers/Rocket_Launcher_Medium_ShotC01}}
 {id=WavePlayer_7, properties {soundWaveAssetPtr=/Game/Sounds/Fort_Weapons/Launchers/Rocket_Launcher_Distant_ShotE01}}
]
properties {
 maxDistance=10579.2871
 soundClassObject=&Fort_WeaponRanged_Fire_3P_PSM
 volumeMultiplier=1
 concurrencyOverrides {bLimitToOwner=true, maxCount=1, resolutionRule=PreventNew}
 concurrencySet [/Game/Athena/Sounds/Weapons/Guided_Missile/Unvault/GuidedMissile_Unvault_Fire]
}
